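# Physics asset for the RoseDust two-handed pickaxe cosmetic, giving its dangling strap physical bodies and joint limits.
# The handle body carries two capsules, and each of the three dyn_strap bones has one capsule.
# Constraints chain the strap segments back to the handle with limited swing and tight twist.
# The root joint swings asymmetrically and barely twists, while the lower joints use symmetric swing limits that widen toward the tip.
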PhysicsAsset /BRCosmetics/Weapons/FORT_Melee/Pickaxe_RoseDust/Meshes/RoseDust_Axe_Physics
bodies {
 Melee_TwoHanded_Handle [
  {shapeType=Capsule, length=4.2246, radius=3.271, shapeName=Handle_capsule_end, center=<0.8956 -1.5791 -39.807>, rotation=<rot 0 0 -11.7732>}
  {shapeType=Capsule, length=5.3043, radius=5.8328, shapeName=Melee_TwoHanded_Handle_capsule, center=<5.0193 -3.3953 -37.3856>}
 ]
 dyn_strap_1 [{shapeType=Capsule, length=2.7452, radius=1.0774, shapeName=dyn_strap_1_capsule, center=<2.6755 0.1798 0>, rotation=<rot 90 180 -177.5572>}]
 dyn_strap_2 [{shapeType=Capsule, length=4.8357, radius=1.1257, shapeName=dyn_strap_2_capsule, center=<3.9191 0 0>, rotation=<rot 90 0 0>}]
 dyn_strap_3 [{
   shapeType=Capsule
   length=3.1477
   radius=1.7881
   shapeName=dyn_strap_3_capsule
   center=<4.3215 -0.7927 -0.0928>
   rotation=<rot 90 180 178.4845>
  }]
}
constraints [
 {bone1Name=dyn_strap_1, bone2Name=Melee_TwoHanded_Handle, swing1LimitDegrees=40, swing2LimitDegrees=140, twistLimitDegrees=2}
 {bone1Name=dyn_strap_2, bone2Name=dyn_strap_1, swing1LimitDegrees=45, swing2LimitDegrees=45, twistLimitDegrees=10}
 {bone1Name=dyn_strap_3, bone2Name=dyn_strap_2, swing1LimitDegrees=60, swing2LimitDegrees=60, twistLimitDegrees=10}
]
